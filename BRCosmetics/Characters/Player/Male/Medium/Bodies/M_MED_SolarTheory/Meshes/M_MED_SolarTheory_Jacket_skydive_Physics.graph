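# Physics asset for the SolarTheory male medium jacket's skydive variant, simulating its coat panels and skirt flaps.
# Capsules on pelvis, thighs and spine form the body volume.
# Tiny master-bone capsules, freely linked to spine_02 or spine_01, anchor angle-limited chains for the center, C and D jacket panels and both skirt flaps.
# Free-angle cross-links tie the middle center segment to both middle D segments, coupling those panels.
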PhysicsAsset /BRCosmetics/Characters/Player/Male/Medium/Bodies/M_MED_SolarTheory/Meshes/M_MED_SolarTheory_Jacket_skydive_Physics
bodies {
 MSTR_jacketBa_c [{shapeType=Capsule, length=0.505, radius=0.505, shapeName=MSTR_jacketBa_c_capsule, rotation=<rot 0 -90 0>}]
 MSTR_jacketC_l [{shapeType=Capsule, length=0.505, radius=0.505, shapeName=MSTR_jacketC_l_capsule, rotation=<rot 0 -90 0>}]
 MSTR_jacketC_r [{shapeType=Capsule, length=0.505, radius=0.505, shapeName=MSTR_jacketC_r_capsule, rotation=<rot 0 -90 0>}]
 MSTR_jacketD_l [{shapeType=Capsule, length=0.505, radius=0.505, shapeName=MSTR_jacketD_l_capsule, rotation=<rot 0 -90 0>}]
 MSTR_jacketD_r [{shapeType=Capsule, length=0.505, radius=0.505, shapeName=MSTR_jacketD_r_capsule, rotation=<rot 0 -90 0>}]
 cr_skirtA2_l [{shapeType=Capsule, length=7.125, radius=7.875, shapeName=dyn_jacketD1_l_capsule, center=<7 0 6>, rotation=<rot 0 0 90>}]
 cr_skirtA2_l_MSTR [{shapeType=Capsule, length=0.505, radius=0.505, shapeName=cr_skirtA2_l_MSTR_capsule, rotation=<rot 0 -90 0>}]
 cr_skirtA2_r [{shapeType=Capsule, length=7.125, radius=7.875, shapeName=dyn_jacketD1_l_capsule, center=<7 0 6>, rotation=<rot 0 0 90>}]
 cr_skirtA2_r_MSTR [{shapeType=Capsule, length=0.505, radius=0.505, shapeName=cr_skirtA2_r_MSTR_capsule, rotation=<rot 0 -90 0>}]
 dfrm_pelvis [
  {shapeType=Capsule, length=0.3538, radius=17.1462, shapeName=dfrm_pelvis_capsule, center=<6.5892 -0.2571 0>}
  {shapeType=Capsule, length=4.875, radius=15.125, shapeName=dfrm_pelvis_capsule, center=<-2.8742 -0.4206 0>}
 ]
 dfrm_thigh_l [{
   shapeType=Capsule
   length=30.5
   radius=10.1
   shapeName=dfrm_thigh_l_capsule
   center=<-15.5093 1.0969 1.1641>
   rotation=<rot 90 -179.5524 -2.0598>
  }]
 dfrm_thigh_r [{
   shapeType=Capsule
   length=30.5
   radius=10.1
   shapeName=dfrm_thigh_l_capsule
   center=<15.5093 -1.0969 -1.1318>
   rotation=<rot 90 180 177.4926>
  }]
 dyn_jacketBa1_c [{shapeType=Capsule, length=8, radius=7, shapeName=dyn_jacketD1_l_capsule, center=<7 -6 0>}]
 dyn_jacketBa2_c [{shapeType=Capsule, length=10.75, radius=7.625, shapeName=dyn_jacketD2_l_capsule, center=<5 -6 0>}]
 dyn_jacketBa3_c [{shapeType=Capsule, length=10.75, radius=7.625, shapeName=dyn_jacketD3_l_capsule, center=<5 -6 0>}]
 dyn_jacketC1_l [{shapeType=Capsule, length=8, radius=7, shapeName=dyn_jacketD1_l_capsule, center=<7 0 6>, rotation=<rot 0 0 90>}]
 dyn_jacketC1_r [{shapeType=Capsule, length=8, radius=7, shapeName=dyn_jacketD1_r_capsule, center=<7 0 6>, rotation=<rot 0 0 90>}]
 dyn_jacketC2_l [{shapeType=Capsule, length=13.375, radius=7, shapeName=dyn_jacketD2_l_capsule, center=<5 0 6>, rotation=<rot 0 0 90>}]
 dyn_jacketC2_r [{shapeType=Capsule, length=13.375, radius=7, shapeName=dyn_jacketD2_r_capsule, center=<5 0 6>, rotation=<rot 0 0 90>}]
 dyn_jacketC3_l [{shapeType=Capsule, length=18.5, radius=7.5, shapeName=dyn_jacketD3_l_capsule, center=<5 0 6>, rotation=<rot 0 0 90>}]
 dyn_jacketC3_r [{shapeType=Capsule, length=18.5, radius=7.5, shapeName=dyn_jacketD3_r_capsule, center=<5 0 6>, rotation=<rot 0 0 90>}]
 dyn_jacketD1_l [{shapeType=Capsule, length=8.75, radius=6.25, shapeName=dyn_jacketD1_l_capsule, center=<7 0 6>, rotation=<rot 0 0 90>}]
 dyn_jacketD1_r [{shapeType=Capsule, length=8.75, radius=6.25, shapeName=dyn_jacketD1_r_capsule, center=<7 0 6>, rotation=<rot 0 0 90>}]
 dyn_jacketD2_l [{shapeType=Capsule, length=11.5, radius=6.875, shapeName=dyn_jacketD2_l_capsule, center=<5 0 6>, rotation=<rot 0 0 90>}]
 dyn_jacketD2_r [{shapeType=Capsule, length=11.5, radius=6.875, shapeName=dyn_jacketD2_r_capsule, center=<5 0 6>, rotation=<rot 0 0 90>}]
 dyn_jacketD3_l [{shapeType=Capsule, length=11.5, radius=6.875, shapeName=dyn_jacketD3_l_capsule, center=<5 0 6>, rotation=<rot 0 0 90>}]
 dyn_jacketD3_r [{shapeType=Capsule, length=11.5, radius=6.875, shapeName=dyn_jacketD3_r_capsule, center=<5 0 6>, rotation=<rot 0 0 90>}]
 dyn_skirtA1_l [{shapeType=Capsule, length=8.375, radius=6.625, shapeName=dyn_jacketD2_l_capsule, center=<2 0 6>, rotation=<rot 0 0 90>}]
 dyn_skirtA1_r [{shapeType=Capsule, length=8.375, radius=6.625, shapeName=dyn_jacketD2_l_capsule, center=<2 0 6>, rotation=<rot 0 0 90>}]
 dyn_skirtA2_l [{shapeType=Capsule, length=8.375, radius=6.625, shapeName=dyn_jacketD3_l_capsule, center=<2 0 6>, rotation=<rot 0 0 90>}]
 dyn_skirtA2_r [{shapeType=Capsule, length=8.375, radius=6.625, shapeName=dyn_jacketD3_l_capsule, center=<2 0 6>, rotation=<rot 0 0 90>}]
 spine_01 [{shapeType=Capsule, length=1.375, radius=15.875, shapeName=spine_01_capsule, center=<-1.0994 0 0>, rotation=<rot 0 -0.0002 0>}]
 spine_02 [{shapeType=Capsule, length=0.75, radius=16.25, shapeName=spine_02_capsule}]
 spine_03 [{shapeType=Capsule, length=2.25, radius=14.75, shapeName=spine_03_capsule, center=<-0.8077 -0.1277 0>}]
}
constraints [
 {
  bone1Name=spine_02
  bone2Name=spine_01
  swing1LimitDegrees=45
  swing1Motion=Free
  swing2LimitDegrees=45
  swing2Motion=Free
  twistLimitDegrees=45
  twistMotion=Free
 }
 {
  bone1Name=spine_03
  bone2Name=spine_02
  swing1LimitDegrees=45
  swing1Motion=Free
  swing2LimitDegrees=45
  swing2Motion=Free
  twistLimitDegrees=45
  twistMotion=Free
 }
 {
  bone1Name=dfrm_thigh_l
  bone2Name=dfrm_pelvis
  swing1LimitDegrees=45
  swing1Motion=Free
  swing2LimitDegrees=45
  swing2Motion=Free
  twistLimitDegrees=45
  twistMotion=Free
 }
 {
  bone1Name=dfrm_thigh_r
  bone2Name=dfrm_pelvis
  swing1LimitDegrees=45
  swing1Motion=Free
  swing2LimitDegrees=45
  swing2Motion=Free
  twistLimitDegrees=45
  twistMotion=Free
 }
 {
  bone1Name=MSTR_jacketBa_c
  bone2Name=spine_02
  swing1LimitDegrees=45
  swing1Motion=Free
  swing2LimitDegrees=45
  swing2Motion=Free
  twistLimitDegrees=45
  twistMotion=Free
 }
 {bone1Name=dyn_jacketBa1_c, bone2Name=MSTR_jacketBa_c, swing1LimitDegrees=15, swing2LimitDegrees=10, twistLimitDegrees=15}
 {bone1Name=dyn_jacketBa2_c, bone2Name=dyn_jacketBa1_c, swing1LimitDegrees=15, swing2LimitDegrees=10, twistLimitDegrees=15}
 {bone1Name=dyn_jacketBa3_c, bone2Name=dyn_jacketBa2_c, swing1LimitDegrees=15, swing2LimitDegrees=10, twistLimitDegrees=5}
 {
  bone1Name=MSTR_jacketC_l
  bone2Name=spine_02
  swing1LimitDegrees=45
  swing1Motion=Free
  swing2LimitDegrees=45
  swing2Motion=Free
  twistLimitDegrees=45
  twistMotion=Free
 }
 {bone1Name=dyn_jacketC1_l, bone2Name=MSTR_jacketC_l, swing1LimitDegrees=5, swing2LimitDegrees=25, twistLimitDegrees=5}
 {bone1Name=dyn_jacketC2_l, bone2Name=dyn_jacketC1_l, swing1LimitDegrees=20, swing2LimitDegrees=25, twistLimitDegrees=15}
 {bone1Name=dyn_jacketC3_l, bone2Name=dyn_jacketC2_l, swing1LimitDegrees=20, swing2LimitDegrees=25, twistLimitDegrees=15}
 {
  bone1Name=MSTR_jacketC_r
  bone2Name=spine_02
  swing1LimitDegrees=45
  swing1Motion=Free
  swing2LimitDegrees=45
  swing2Motion=Free
  twistLimitDegrees=45
  twistMotion=Free
 }
 {bone1Name=dyn_jacketC1_r, bone2Name=MSTR_jacketC_r, swing1LimitDegrees=5, swing2LimitDegrees=25, twistLimitDegrees=5}
 {bone1Name=dyn_jacketC2_r, bone2Name=dyn_jacketC1_r, swing1LimitDegrees=20, swing2LimitDegrees=25, twistLimitDegrees=15}
 {bone1Name=dyn_jacketC3_r, bone2Name=dyn_jacketC2_r, swing1LimitDegrees=20, swing2LimitDegrees=25, twistLimitDegrees=15}
 {
  bone1Name=MSTR_jacketD_l
  bone2Name=spine_02
  swing1LimitDegrees=45
  swing1Motion=Free
  swing2LimitDegrees=45
  swing2Motion=Free
  twistLimitDegrees=45
  twistMotion=Free
 }
 {bone1Name=dyn_jacketD1_l, bone2Name=MSTR_jacketD_l, swing1LimitDegrees=15, swing2LimitDegrees=15, twistLimitDegrees=15}
 {bone1Name=dyn_jacketD2_l, bone2Name=dyn_jacketD1_l, swing1LimitDegrees=15, swing2LimitDegrees=15, twistLimitDegrees=15}
 {bone1Name=dyn_jacketD3_l, bone2Name=dyn_jacketD2_l, swing1LimitDegrees=15, swing2LimitDegrees=15, twistLimitDegrees=15}
 {
  bone1Name=MSTR_jacketD_r
  bone2Name=spine_02
  swing1LimitDegrees=45
  swing1Motion=Free
  swing2LimitDegrees=45
  swing2Motion=Free
  twistLimitDegrees=45
  twistMotion=Free
 }
 {bone1Name=dyn_jacketD1_r, bone2Name=MSTR_jacketD_r, swing1LimitDegrees=15, swing2LimitDegrees=15, twistLimitDegrees=15}
 {bone1Name=dyn_jacketD2_r, bone2Name=dyn_jacketD1_r, swing1LimitDegrees=15, swing2LimitDegrees=15, twistLimitDegrees=15}
 {bone1Name=dyn_jacketD3_r, bone2Name=dyn_jacketD2_r, swing1LimitDegrees=15, swing2LimitDegrees=15, twistLimitDegrees=15}
 {
  bone1Name=dyn_jacketBa2_c
  bone2Name=dyn_jacketD2_r
  swing1LimitDegrees=45
  swing1Motion=Free
  swing2LimitDegrees=45
  swing2Motion=Free
  twistLimitDegrees=45
  twistMotion=Free
 }
 {
  bone1Name=dyn_jacketBa2_c
  bone2Name=dyn_jacketD2_l
  swing1LimitDegrees=45
  swing1Motion=Free
  swing2LimitDegrees=45
  swing2Motion=Free
  twistLimitDegrees=45
  twistMotion=Free
 }
 {bone1Name=cr_skirtA2_l, bone2Name=cr_skirtA2_l_MSTR, swing1LimitDegrees=10, swing2LimitDegrees=25, twistLimitDegrees=5}
 {bone1Name=dyn_skirtA1_l, bone2Name=cr_skirtA2_l, swing1LimitDegrees=10, swing2LimitDegrees=25, twistLimitDegrees=15}
 {bone1Name=dyn_skirtA2_l, bone2Name=dyn_skirtA1_l, swing1LimitDegrees=10, swing2LimitDegrees=25, twistLimitDegrees=15}
 {bone1Name=cr_skirtA2_r, bone2Name=cr_skirtA2_r_MSTR, swing1LimitDegrees=10, swing2LimitDegrees=25, twistLimitDegrees=5}
 {bone1Name=dyn_skirtA1_r, bone2Name=cr_skirtA2_r, swing1LimitDegrees=10, swing2LimitDegrees=25, twistLimitDegrees=15}
 {bone1Name=dyn_skirtA2_r, bone2Name=dyn_skirtA1_r, swing1LimitDegrees=10, swing2LimitDegrees=25, twistLimitDegrees=15}
 {
  bone1Name=cr_skirtA2_r_MSTR
  bone2Name=spine_01
  swing1LimitDegrees=45
  swing1Motion=Free
  swing2LimitDegrees=45
  swing2Motion=Free
  twistLimitDegrees=45
  twistMotion=Free
 }
 {
  bone1Name=cr_skirtA2_l_MSTR
  bone2Name=spine_01
  swing1LimitDegrees=45
  swing1Motion=Free
  swing2LimitDegrees=45
  swing2Motion=Free
  twistLimitDegrees=45
  twistMotion=Free
 }
]
