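# Post-process style AnimBP for the SolarTheory male medium cosmetic body, animating jacket and skirt secondary motion on the pose copied from the parent mesh.
# Weighted bone constraints and a state-driven Control Rig feed two RigidBody sims with overridden gravity, disabled while skydiving and bypassed while surface swimming.
# Additive bone offsets adjust crouch-walk and jumping, and a front-end curve layers DefaultSlot onto hand metacarpals.
# Some constraint and Control Rig nodes are left unconnected.
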AnimBlueprint /BRCosmetics/Characters/Player/Male/Medium/Bodies/M_MED_SolarTheory/Meshes/M_MED_SolarTheory_AnimBP
target_skeleton=/Game/Characters/Player/Male/Medium/Bodies/M_MED_SolarTheory/Meshes/M_MED_SolarTheory_Skeleton
graphs [
 AnimGraph {nodes [
   n1 = "Constraint - base_skirtB_l (PP)"!(ComponentPose, Alpha=1, ConstraintWeights_0=1, ConstraintWeights_1:float=1) -> (Pose) {
    alphaBoolBlend {blendOption=Linear}
    boneToModify {boneName=base_skirtB_l}
    constraintSetup [{targetBone {boneName=dyn_skirtA1_l}}, {targetBone {boneName=dyn_jacketC3_l}}]
    constraintWeights [1,1]
   }
   "Constraint - base_skirtB_r (PP)"!(ComponentPose=@n1, Alpha=1, ConstraintWeights_0=1, ConstraintWeights_1:float=1) -> (Pose) {
    alphaBoolBlend {blendOption=Linear}
    boneToModify {boneName=base_skirtB_r}
    constraintSetup [{targetBone {boneName=base_skirtA_r}}, {targetBone {boneName=dyn_jacketC3_r}}]
    constraintWeights [1,1]
   }
   n2 = Control Rig() {
    alphaBoolBlend {blendOption=Linear}
    controlRigAssetReference {blueprintRigClass=/Game/Characters/Player/Male/Medium/Bodies/M_MED_SolarTheory/Meshes/M_MED_SolarTheory_CtrlRig.M_MED_SolarTheory_CtrlRig_C}
   }
   n3 = Control Rig(Source=@n2) {
    alphaBoolBlend {blendOption=Linear}
    controlRigAssetReference {blueprintRigClass=/Game/Characters/Player/Male/Medium/Bodies/M_MED_SolarTheory/Meshes/M_MED_SolarTheory_ControlRig_Twist1.M_MED_SolarTheory_ControlRig_Twist1_C}
   }
   Control Rig(Source=@n3) {
    alphaBoolBlend {blendOption=Linear}
    controlRigAssetReference {blueprintRigClass=/Game/Characters/Player/Male/Medium/Bodies/M_MED_SolarTheory/Meshes/M_MED_SolarTheory_DeformRig.M_MED_SolarTheory_DeformRig_C}
   }
   n4 = Layered blend per bone(BasePose=Use cached pose 'FinalDynPose'(), BlendPoses_0=Slot 'DefaultSlot'(), BlendWeights_0=Get Frontend Idle() -> (Frontend Idle:double)) {layerSetup [{branchFilters [{boneName=middle_metacarpal_l}, {boneName=middle_metacarpal_r}, {boneName=ring_metacarpal_r}, {boneName=ring_metacarpal_l}]}]}
   Output Pose(Result=@n4) {layerGroup=None}
   n5 = Get NotInSkyDiving() -> (NotInSkyDiving:bool)
   n6 = Get GravityOverride()
   n7 = RigidBody!(
    ComponentPose=Local To Component(LocalPose=Use cached pose 'RigidBody'())
    bAlphaBoolEnabled:bool=@n5
    OverrideWorldGravity:Vector=@n6
    ExternalForce
   ) -> (Pose) {
    alphaInputType=Bool
    bClampLinearTranslationLimitToRefPose=true
    bOverrideWorldGravity=true
    overridePhysicsAsset=&/M_MED_SolarTheory_Jacket_skydive_Physics
    simulationSpace=BaseBoneSpace
    alphaBoolBlend {blendInTime=0.1, blendOption=Linear, blendOutTime=0.1}
    baseBoneRef {boneName=pelvis}
    componentAppliedLinearAccClamp=<500 500 500>
    componentLinearAccScale=<0.3 0.3 0.3>
    componentLinearVelScale=<0.3 0.3 0.3>
    overrideWorldGravity {z=-980}
   }
   n8 = RigidBody(ComponentPose=@n7, bAlphaBoolEnabled=@n5, OverrideWorldGravity=@n6) {
    alphaInputType=Bool
    bClampLinearTranslationLimitToRefPose=true
    bOverrideWorldGravity=true
    overridePhysicsAsset=&/M_MED_SolarTheory_Jacket_Physics
    simulationSpace=BaseBoneSpace
    alphaBoolBlend {blendInTime=0.1, blendOption=Linear, blendOutTime=0.1}
    baseBoneRef {boneName=pelvis}
    componentAppliedLinearAccClamp=<1000 1000 1000>
    componentLinearAccScale=<0.3 0.3 0.3>
    componentLinearVelScale=<0.3 0.3 0.3>
    overrideWorldGravity {z=-980}
   }
   n9 = Component To Local(ComponentPose=@n8)
   n10 = Use cached pose 'RigidBody'()
   n11 = Get NotInSurfaceSwimming() -> (NotInSurfaceSwimming:bool)
   n12 = Get crouch walk() -> (crouch walk:double)
   n13 = "Transform (Modify) Bone - Bone: dyn_jacketBa1_c"(
    ComponentPose=Local To Component(LocalPose=Blend Poses by bool(BlendPose_0=@n9, BlendPose_1=@n10, bActiveValue=@n11) {blendType=Linear})
    Alpha=@n12
    Translation="4,0,0"
    Rotation="0,0,0"
    Scale="1,1,1"
   ) {
    rotationMode=BMM_Additive
    rotationSpace=BCS_BoneSpace
    translationMode=BMM_Additive
    translationSpace=BCS_BoneSpace
    alphaBoolBlend {blendOption=Linear}
    alphaScaleBiasClamp {bInterpResult=true, interpSpeedIncreasing=2}
    boneToModify {boneName=dyn_jacketBa1_c}
    translation {x=4}
   }
   n14 = "Transform (Modify) Bone - Bone: dyn_jacketD1_l"(ComponentPose=@n13, Alpha=@n12, Translation="2,0,0", Rotation="0,0,0", Scale="1,1,1") {
    rotationMode=BMM_Additive
    rotationSpace=BCS_BoneSpace
    translationMode=BMM_Additive
    translationSpace=BCS_BoneSpace
    alphaBoolBlend {blendOption=Linear}
    alphaScaleBiasClamp {bInterpResult=true, interpSpeedIncreasing=2}
    boneToModify {boneName=dyn_jacketD1_l}
    translation {x=2}
   }
   n15 = "Transform (Modify) Bone - Bone: dyn_jacketD1_r"(ComponentPose=@n14, Alpha=@n12, Translation="2,0,0", Rotation="0,0,0", Scale="1,1,1") {
    rotationMode=BMM_Additive
    rotationSpace=BCS_BoneSpace
    translationMode=BMM_Additive
    translationSpace=BCS_BoneSpace
    alphaBoolBlend {blendOption=Linear}
    alphaScaleBiasClamp {bInterpResult=true, interpSpeedIncreasing=2}
    boneToModify {boneName=dyn_jacketD1_r}
    translation {x=2}
   }
   n16 = "Transform (Modify) Bone - Bone: cr_skirtA2_r_MSTR"(ComponentPose=@n15, Alpha=Get jumping() -> (jumping:double), Translation="-1,3,3", Rotation="0,25,35", Scale="1,1,1") {
    rotationMode=BMM_Additive
    rotationSpace=BCS_BoneSpace
    translationMode=BMM_Additive
    translationSpace=BCS_BoneSpace
    alphaBoolBlend {blendOption=Linear}
    alphaScaleBiasClamp {bInterpResult=true, interpSpeedIncreasing=2}
    boneToModify {boneName=cr_skirtA2_r_MSTR}
    rotation {roll=35, yaw=25}
    translation=<-1 3 3>
   }
   Save cached pose 'FinalDynPose'(Pose=Component To Local(ComponentPose=@n16)) {cacheName=FinalDynPose}
   n17 = "Constraint - MSTR_jacketBa_c (PPPPPPPPP)"!(
    ComponentPose=Local To Component(LocalPose=Copy Pose From Mesh() {bUseAttachedParent=true})
    Alpha=1
    ConstraintWeights_0=0.0031
    ConstraintWeights_1:float=0.1288
    ConstraintWeights_2:float=0.7459
    ConstraintWeights_3:float=0
    ConstraintWeights_4:float=0.0061
    ConstraintWeights_5:float=0.0061
    ConstraintWeights_6:float=0.0551
    ConstraintWeights_7:float=0.0549
    ConstraintWeights_8:float=0
   ) -> (Pose) {
    alphaBoolBlend {blendOption=Linear}
    boneToModify {boneName=MSTR_jacketBa_c}
    constraintSetup [
     {targetBone {boneName=spine_01}}
     {targetBone {boneName=spine_02}}
     {targetBone {boneName=spine_03}}
     {targetBone {boneName=dfrm_pelvis}}
     {targetBone {boneName=cr_skirtB_r}}
     {targetBone {boneName=cr_skirtB2_l}}
     {targetBone {boneName=cr_skirtB1_r}}
     {targetBone {boneName=cr_skirtB1_l}}
     {targetBone {boneName=base_skirtB_r}}
    ]
    constraintWeights [1,1,1,1,1,1,1,1,1]
   }
   n18 = "Constraint - MSTR_jacketD_l (PPPPPPPPPP)"!(
    ComponentPose=@n17
    Alpha=1
    ConstraintWeights_0=0
    ConstraintWeights_1:float=0.0001
    ConstraintWeights_2:float=0.0478
    ConstraintWeights_3:float=0.0029
    ConstraintWeights_4:float=0.6683
    ConstraintWeights_5:float=0
    ConstraintWeights_6:float=0
    ConstraintWeights_7:float=0.0275
    ConstraintWeights_8:float=0.2534
    ConstraintWeights_9:float=0
   ) -> (Pose) {
    alphaBoolBlend {blendOption=Linear}
    boneToModify {boneName=MSTR_jacketD_l}
    constraintSetup [
     {targetBone {boneName=spine_01}}
     {targetBone {boneName=spine_02}}
     {targetBone {boneName=spine_04}}
     {targetBone {boneName=spine_05}}
     {targetBone {boneName=spine_03}}
     {targetBone {boneName=dfrm_pelvis}}
     {targetBone {boneName=dfrm_upperarm_twist_01_l}}
     {targetBone {boneName=cr_skirtB2_l}}
     {targetBone {boneName=cr_skirtB1_l}}
     {targetBone {boneName=base_skirtB_l}}
    ]
    constraintWeights [1,1,1,1,1,1,1,1,1,1]
   }
   n19 = "Constraint - MSTR_jacketD_r (PPPPPPPPPP)"!(
    ComponentPose=@n18
    Alpha=1
    ConstraintWeights_0=0
    ConstraintWeights_1:float=0.0001
    ConstraintWeights_2:float=0.0478
    ConstraintWeights_3:float=0.0029
    ConstraintWeights_4:float=0.6683
    ConstraintWeights_5:float=0
    ConstraintWeights_6:float=0
    ConstraintWeights_7:float=0.0275
    ConstraintWeights_8:float=0.2534
    ConstraintWeights_9:float=0
   ) -> (Pose) {
    alphaBoolBlend {blendOption=Linear}
    boneToModify {boneName=MSTR_jacketD_r}
    constraintSetup [
     {targetBone {boneName=spine_01}}
     {targetBone {boneName=spine_02}}
     {targetBone {boneName=spine_04}}
     {targetBone {boneName=spine_05}}
     {targetBone {boneName=spine_03}}
     {targetBone {boneName=dfrm_pelvis}}
     {targetBone {boneName=dfrm_upperarm_twist_01_r}}
     {targetBone {boneName=cr_skirtB_r}}
     {targetBone {boneName=cr_skirtB1_r}}
     {targetBone {boneName=base_skirtB_r}}
    ]
    constraintWeights [1,1,1,1,1,1,1,1,1,1]
   }
   n20 = "Constraint - MSTR_jacketC_l (PPPPPPPPPP)"!(
    ComponentPose=@n19
    Alpha=1
    ConstraintWeights_0=0.0526
    ConstraintWeights_1:float=0.0027
    ConstraintWeights_2:float=0.4634
    ConstraintWeights_3:float=0
    ConstraintWeights_4:float=0.0077
    ConstraintWeights_5:float=0.0694
    ConstraintWeights_6:float=0
    ConstraintWeights_7:float=0.0369
    ConstraintWeights_8:float=0.3672
    ConstraintWeights_9:float=0
   ) -> (Pose) {
    alphaBoolBlend {blendOption=Linear}
    boneToModify {boneName=MSTR_jacketC_l}
    constraintSetup [
     {targetBone {boneName=spine_04}}
     {targetBone {boneName=spine_05}}
     {targetBone {boneName=spine_03}}
     {targetBone {boneName=dfrm_upperarm_twist_01_l}}
     {targetBone {boneName=cr_skirtA2_l}}
     {targetBone {boneName=cr_skirtA1_l}}
     {targetBone {boneName=base_skirtA_l}}
     {targetBone {boneName=cr_skirtB2_l}}
     {targetBone {boneName=cr_skirtB1_l}}
     {targetBone {boneName=base_skirtB_l}}
    ]
    constraintWeights [1,1,1,1,1,1,1,1,1,1]
   }
   n21 = "Constraint - MSTR_jacketC_r (PPPPPPPPPP)"!(
    ComponentPose=@n20
    Alpha=1
    ConstraintWeights_0=0.0526
    ConstraintWeights_1:float=0.0027
    ConstraintWeights_2:float=0.4634
    ConstraintWeights_3:float=0
    ConstraintWeights_4:float=0.0369
    ConstraintWeights_5:float=0.0077
    ConstraintWeights_6:float=0.3672
    ConstraintWeights_7:float=0.0694
    ConstraintWeights_8:float=0
    ConstraintWeights_9:float=0
   ) -> (Pose) {
    alphaBoolBlend {blendOption=Linear}
    boneToModify {boneName=MSTR_jacketC_r}
    constraintSetup [
     {targetBone {boneName=spine_04}}
     {targetBone {boneName=spine_05}}
     {targetBone {boneName=spine_03}}
     {targetBone {boneName=dfrm_upperarm_twist_01_r}}
     {targetBone {boneName=cr_skirtB_r}}
     {targetBone {boneName=cr_skirtA2_r}}
     {targetBone {boneName=cr_skirtB1_r}}
     {targetBone {boneName=cr_skirtA1_r}}
     {targetBone {boneName=base_skirtA_r}}
     {targetBone {boneName=base_skirtB_r}}
    ]
    constraintWeights [1,1,1,1,1,1,1,1,1,1]
   }
   n22 = "Constraint - base_skirtB_l (PPPPPPPPP)"!(
    ComponentPose=@n21
    Alpha=1
    ConstraintWeights_0=0.0271
    ConstraintWeights_1:float=0.1562
    ConstraintWeights_2:float=0.068
    ConstraintWeights_3:float=0.1507
    ConstraintWeights_4:float=0.0943
    ConstraintWeights_5:float=0.0219
    ConstraintWeights_6:float=0.074
    ConstraintWeights_7:float=0.1766
    ConstraintWeights_8:float=0.2313
   ) -> (Pose) {
    alphaBoolBlend {blendOption=Linear}
    boneToModify {boneName=base_skirtB_l}
    constraintSetup [
     {targetBone {boneName=cr_skirtA2_l}}
     {targetBone {boneName=base_skirtA_l}}
     {targetBone {boneName=dyn_skirtA1_l}}
     {targetBone {boneName=dyn_skirtA2_l}}
     {targetBone {boneName=base_skirtB_l}}
     {targetBone {boneName=dyn_skirtB1_l}}
     {targetBone {boneName=dyn_skirtB2_l}}
     {targetBone {boneName=dyn_jacketC2_l}}
     {targetBone {boneName=dyn_jacketC3_l}}
    ]
    constraintWeights [1,1,1,1,1,1,1,1,1]
   }
   n23 = "Constraint - base_skirtB_r (PPPPPPPPP)"!(
    ComponentPose=@n22
    Alpha=1
    ConstraintWeights_0=0.0271
    ConstraintWeights_1:float=0.1562
    ConstraintWeights_2:float=0.068
    ConstraintWeights_3:float=0.1507
    ConstraintWeights_4:float=0.0943
    ConstraintWeights_5:float=0.0219
    ConstraintWeights_6:float=0.074
    ConstraintWeights_7:float=0.1766
    ConstraintWeights_8:float=0.2313
   ) -> (Pose) {
    alphaBoolBlend {blendOption=Linear}
    boneToModify {boneName=base_skirtB_r}
    constraintSetup [
     {targetBone {boneName=cr_skirtA2_r}}
     {targetBone {boneName=base_skirtA_r}}
     {targetBone {boneName=dyn_skirtA1_r}}
     {targetBone {boneName=dyn_skirtA2_r}}
     {targetBone {boneName=base_skirtB_r}}
     {targetBone {boneName=dyn_skirtB1_r}}
     {targetBone {boneName=dyn_skirtB2_r}}
     {targetBone {boneName=dyn_jacketC2_r}}
     {targetBone {boneName=dyn_jacketC3_r}}
    ]
    constraintWeights [1,1,1,1,1,1,1,1,1]
   }
   n24 = Control Rig!(
    Source=Component To Local(ComponentPose=@n23)
    Alpha
    crouch:double=Get crouch() -> (crouch:double)
    moving:double=Get moving2d() -> (moving2d:double)
    sprint:double=Get sprint() -> (sprint:double)
   ) -> (Pose) {
    alphaBoolBlend {blendOption=Linear}
    controlRigAssetReference {blueprintRigClass=&/M_MED_SolarTheory_All.M_MED_SolarTheory_All_C}
   }
   Save cached pose 'RigidBody'(Pose=@n24) {cacheName=RigidBody}
  ]}
 EventGraph {nodes [
   Event BlueprintUpdateAnimation()
   > n1 = Sequence!(execute) -> (then_0, then_1, then_2:Exec, then_3:Exec)
   > n2 = Gravity Override State Function DEPRECATED(CharacterPartAnimInstance=Self-Reference(), GameTimeInSeconds=GetGameTimeInSeconds(), GravityOverrideParams=Get Gravity Override Params())
   > Set GravityOverride(GravityOverride=@n2)
   > n3 = Branch(Condition=OR Boolean(A=Get bIsJumping(), B=Get bIsFalling()))
   > Set jumping(jumping:double=1) -> (Output_Get:double)
   Set jumping(execute=@n3.else, jumping=0)
   n4 = Branch(execute=@n1.then_1, Condition=AND Boolean(A=Get bIsCrouching(), B=Get bIsMoving2D()))
   > Set crouch walk(crouch walk:double=1) -> (Output_Get:double)
   Set crouch walk(execute=@n4.else, crouch walk=0)
   Set crouch(execute=@n1.then_2, crouch:double="To Float (Boolean)"(InBool=Get bIsCrouching())) -> (Output_Get:double)
   > Set moving2d(moving2d:double="To Float (Boolean)"(InBool=Get bIsMoving2D())) -> (Output_Get:double)
   > Set sprint(sprint:double="To Float (Boolean)"(InBool=Get bIsSprinting())) -> (Output_Get:double)
   > Set NotInSkyDiving(NotInSkyDiving:bool=NOT Boolean(A=Get bIsSkyDiving())) -> (Output_Get:bool)
   > Set NotInSurfaceSwimming(NotInSurfaceSwimming:bool=NOT Boolean(A=Get bIsSurfaceSwimming())) -> (Output_Get:bool)
   Branch(execute=@n1.then_3, Condition=Get bIsInFrontEnd())
   > Set Frontend Idle(Frontend Idle:double=GetCurveValue(self=GetOwnerFrontEndAnimInstance(), CurveName=Frontend_Idle_Bypass)) -> (Output_Get:double)
  ]}
]
variables [
 crouch {type=number}
 crouch walk {type=number}
 Frontend Idle {type=number}
 Gravity Override Params {
  type=GravityOverrideParamsStruct
  default {skydive_FloatingDown {gravityOverride=<0 -125 -400>, windAmplitude=<50 -125 100>, windFrequency=<0.75 2 2.5>, windOffset=<0 -150 100>}}
 }
 GravityOverride {type=Vector}
 jumping {type=number}
 moving2d {type=number}
 NotInSkyDiving {type=boolean}
 NotInSurfaceSwimming {type=boolean}
 sprint {type=number}
]
