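# Physics asset for the face accessory mesh of the male medium SolarTheory player cosmetic, covering front and back hair bones.
# Single capsules sit on HairFr1, HairBaA1, the two dyn_hair bones and two tiny MSTR anchor bones.
# Two constraints tie each dyn_hair bone to its MSTR bone with very tight swing limits and zero twist limit, keeping any sway small.
# HairFr1 and HairBaA1 carry larger capsules with no constraints.
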PhysicsAsset /BRCosmetics/Characters/Player/Male/Medium/Bodies/M_MED_SolarTheory/Meshes/Parts/M_MED_SolarTheory_FaceAcc_Physics
bodies {
 HairBaA1 [{
   shapeType=Capsule
   length=15.7288
   radius=14.916
   shapeName=HairBaA1_capsule
   center=<1.3397 -2.2789 0.5028>
   rotation=<rot 0 -48.8014 -71.7669>
  }]
 HairFr1 [{
   shapeType=Capsule
   length=12.0876
   radius=7.3599
   shapeName=HairFr1_capsule
   center=<8.8681 -0.602 0.6266>
   rotation=<rot 0 89.1905 -54.0195>
  }]
 dyn_hair_Fr [{shapeType=Capsule, length=10, radius=8, shapeName=dyn_hair_Fr_capsule, center=<7.5 0 0>, rotation=<rot 90 0 0>}]
 dyn_hair_ba [{shapeType=Capsule, length=7, radius=4.5, shapeName=dyn_hair_ba_capsule, center=<4 0 0>, rotation=<rot 90 0 0>}]
 hair_Ba_MSTR [{shapeType=Capsule, length=0.5, radius=0.5, shapeName=hair_Ba_MSTR_capsule}]
 hair_Fr_MSTR [{shapeType=Capsule, length=0.5, radius=0.5, shapeName=hair_Fr_MSTR_capsule}]
}
constraints [
 {bone1Name=dyn_hair_ba, bone2Name=hair_Ba_MSTR, swing1LimitDegrees=0, swing2LimitDegrees=3, twistLimitDegrees=0}
 {bone1Name=dyn_hair_Fr, bone2Name=hair_Fr_MSTR, swing1LimitDegrees=1, swing2LimitDegrees=2, twistLimitDegrees=0}
]
